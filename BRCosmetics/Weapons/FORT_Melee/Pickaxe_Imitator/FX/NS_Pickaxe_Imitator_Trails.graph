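# Niagara system that draws the swing trail for the Imitator pickaxe as a single ribbon emitter.
# AnimTrail modules spawn short-lived grey ribbon particles from bone or socket indices, with a skeletal mesh data interface exposed as a user parameter.
# Each update writes dynamic material parameters, two from random ranges, and scales alpha by time of day so the trail is brightest by day.
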NiagaraSystem /BRCosmetics/Weapons/FORT_Melee/Pickaxe_Imitator/FX/NS_Pickaxe_Imitator_Trails
dependencies {usedDataInterfaces [], usedDynamicInputs [], usedModules [&SystemState]}
emitters {AnimTrail {
  emitterUpdateScript [
   &EmitterState()
   &FN_SpawnAnimTrail_v2(Spawn Interval=0)
  ]
  particleSpawnScript [
   /Game/Effects/Niagara/Modules/AnimTrail/FN_InitializeAnimTrail/FN_InitializeAnimTrailBoneIndecies(Second Bone or Socket Index=1)
   /Game/Effects/Niagara/Modules/AnimTrail/FN_InitializeAnimTrail/FN_InitializeAnimTrail_v2(AnimTrailWidth=1, Color:LinearColor=<rgba 0.3074 0.3074 0.3438 1>, Lifetime=0.225)
  ]
  particleUpdateScript [
   &ParticleState()
   &FN_UpdateAnimTrail_v2()
   &SolveForcesAndVelocity(Clamp Velocity=false, Limit Acceleration=false)
   &DynamicMaterialParameters(
    Index 0 Param 1=&UniformRangedFloat(Minimum=0, Maximum=1)
    Index 0 Param 2=&UniformRangedFloat(Minimum=0, Maximum=1)
    Index 0 Param 3=1
    Index 0 Param 4=1
    Module.Param0WriteEnabled=true
    Module.Param1WriteEnabled=true
    Module.Param2WriteEnabled=true
    Module.Param3WriteEnabled=true
   )
   &ScaleColor(
    Scale Alpha=&FNTimeOfDayScalar(Night Scale=0.5, Morning Scale=0.5, Day Scale=1.1, Evening Scale=0.3)
    ScaleA=true
    ScaleRGB=true
   )
  ]
  rendererClasses [&NiagaraRibbonRendererProperties]
  renderers [{rendererClass=&NiagaraRibbonRendererProperties}]
 }}
summary {
 systemName=NS_Pickaxe_Imitator_Trails
 emitters [{emitterName=AnimTrail, rendererClasses [&NiagaraRibbonRendererProperties]}]
 userVariables [
  User.ColorSwitcher {type=NiagaraFloat, defaultValue=0, description=""}
  User.RibbonIDOffset {type=NiagaraInt32}
  User.SkeletalMesh {
   type=NiagaraDataInterfaceSkeletalMesh
   defaultValue {
    struct=NiagaraExt_VariableValue_DataInterface
    value {dataInterface=&self:NiagaraDataInterfaceSkeletalMesh_0}
   }
  }
 ]
}
